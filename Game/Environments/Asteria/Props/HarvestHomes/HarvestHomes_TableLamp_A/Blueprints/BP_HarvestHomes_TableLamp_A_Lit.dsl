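; Blueprint for the lit variant of the Harvest Homes table lamp prop in the Asteria environment.
; It supplies fixed time-of-day default variables through an override function, and its construction script only calls the parent's.
; BeginPlay, overlap and tick events are empty, as is a stub NewFunction, so any lighting behavior comes from outside this graph.
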(event EventBeginPlay)

(event Collision|EventActorBeginOverlap
  (params (OtherActor :type Actor%20Object%20Reference)))

(event EventTick
  (params (DeltaSeconds :type Float%20%28single-precision%29)))

(fn GetTimeOfDayBlueprintDefaultVariables
  (return (Utilities|Struct|MakeTimeOfDayBlueprintDefaultVariables 0 1.0 false true)))

(fn NewFunction)

(fn ConstructionScript
  (|CallParentFunction))
